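; Melee Mantis gameplay ability that alternates Vader between his two combos on successive uses.
; On activation it runs the parent logic and caches the avatar as the player pawn.
; On every end, cancelled or not, it toggles the MeleeTags_Combo01 loose gameplay tags on that pawn, removing them if any are present and adding them otherwise.
; The MantisData reads these tags to choose the combo.
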(event Ability|EventActivateAbility
  (|CallParentFunction)
  (Variables|Default|SetPlayerPawn (CallFunction|GetAvatarasFortPlayerPawn)))

(event Ability|EventOnEndAbility
  (params (bWasCancelled :type Boolean))
  (|CallParentFunction bWasCancelled)
  (CallFunction|ValidateMeleeTagforCombo))

(event Custom|ValidateMeleeTagForCombo
  ;[1 1 1] Assign Tags that will be read from the MantisData, makes Vader alternate between his two combos.
  (if (GameplayTags|HasAnyMatchingGameplayTags (Variables|Default|GetPlayerPawn) (Variables|Default|GetMeleeTags_Combo01))
    (Ability|GameplayEffect|RemoveLooseGameplayTags (Variables|Default|GetPlayerPawn) (Variables|Default|GetMeleeTags_Combo01))
    (else
      (Ability|GameplayEffect|AddLooseGameplayTags (Variables|Default|GetPlayerPawn) (Variables|Default|GetMeleeTags_Combo01)) ; Should ever only hit this once
    )))
